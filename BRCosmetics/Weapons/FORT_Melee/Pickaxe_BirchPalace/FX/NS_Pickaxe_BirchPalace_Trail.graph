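# Niagara swing trail for the BirchPalace pickaxe, built from layered ribbons and sparks.
# Four AnimTrail ribbon emitters, including dark and light layers, follow bone or socket pairs and fade over short lifetimes.
# A sprite emitter spawns looping spark bursts around a socket, drifting with curl noise, gravity and drag.
# User.Variant blends warm palettes toward blue, while time of day, front-end context and System.AlphaFade modulate trail color or opacity.
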NiagaraSystem /BRCosmetics/Weapons/FORT_Melee/Pickaxe_BirchPalace/FX/NS_Pickaxe_BirchPalace_Trail
dependencies {
 usedDataInterfaces [&NiagaraDataInterfaceCurve, &NiagaraDataInterfaceColorCurve, &NiagaraDataInterfaceSocketReader]
 usedDynamicInputs [&SetBoolByFloatComparison, &MergeHelper_FloatSwitch]
 usedModules [
  &self:SystemSpawnScript.SystemScriptSource.SystemScriptGraph.NiagaraNodeAssignment_0.SetVariables_#1
  &self:SystemSpawnScript.SystemScriptSource.SystemScriptGraph.NiagaraNodeAssignment_3.SetVariables_#2
  &self:SystemSpawnScript.SystemScriptSource.SystemScriptGraph.NiagaraNodeAssignment_2.SetVariables_#3
  &SystemState
  &self:SystemSpawnScript.SystemScriptSource.SystemScriptGraph.NiagaraNodeAssignment_1.SetVariables_#4
 ]
}
emitters {
 FN_AnimTrail {
  emitterSpawnScript [&FN_RibbonIDOffsetUtility()]
  emitterUpdateScript [
   &EmitterState()
   &FN_SpawnAnimTrail()
  ]
  particleSpawnScript [
   &FN_InitializeAnimTrailBoneIndecies(Second Bone or Socket Index=2)
   &FN_InitializeAnimTrailBoneIndecies(AnimTrail Selection=1, First Bone or Socket Index=1, Second Bone or Socket Index=0) {enabled=false, module=FN_InitializeAnimTrailBoneIndecies001}
   &FN_InitializeAnimTrail(AnimTrailWidth=0.97, Lifetime=0.3)
   &AddVelocity(Velocity=<0 400 0>) {enabled=false}
  ]
  particleUpdateScript [
   &SolveForcesAndVelocity(Clamp Velocity=false, Limit Acceleration=false) {enabled=false}
   &FN_UpdateAnimTrail() {module=FN_UpdateAnimTrail001}
   &ScaleColor(
    Scale Alpha=&FloatFromCurve(
     FloatCurve=DataInterface() {propertyValues {Curve {keys [
         {arriveTangent=-0.4618, interpMode=RCIM_Cubic, leaveTangent=-0.4618, tangentMode=RCTM_User, time=0, value=1}
         {arriveTangent=-1.1299, interpMode=RCIM_Cubic, leaveTangent=-1.1299, tangentMode=RCTM_User, time=0.4269, value=0.2853}
         {arriveTangent=-0.0168, interpMode=RCIM_Cubic, tangentMode=RCTM_Break, time=1, value=0}
        ]}}}
     Scale Curve=&Multiply_Float(
      A=&FN_FrontEndFloatModifier(Gameplay=0.5, Front End=0.3)
      B=&FNTimeOfDayScalar(Night Scale=0.5, Morning Scale=0.6, Evening Scale=0.6)
     )
    )
    Scale RGB=&MakeVectorFromLinearColorRGB(Color=<rgba 1 1 1 1>)
    ScaleA=true
    ScaleRGB=true
   ) {enabled=false, module=ScaleColor002}
   &ScaleColor(
    Scale Alpha=&Multiply_Float(
     A=&FloatFromCurve(
      FloatCurve=DataInterface() {propertyValues {Curve {keys [
          <at 0 1.7371 RCIM_Cubic>
          {arriveTangent=-1.3606, interpMode=RCIM_Cubic, leaveTangent=-1.3606, time=0.2132, value=1.2596}
          {arriveTangent=-0.4634, interpMode=RCIM_Cubic, leaveTangent=-0.4634, tangentMode=RCTM_User, time=0.4416, value=1.1363}
          {arriveTangent=-0.0464, interpMode=RCIM_Cubic, tangentMode=RCTM_Break, time=1, value=0.5443}
         ]}}}
      Scale Curve=&Multiply_Float(
       A=&Lerp_Float(A=1, B=0.75, Alpha=$User.Variant)
       B=&FN_FrontEndFloatModifier(Gameplay=1, Front End=0.8)
      )
     )
     B=&Multiply_Float(
      A=&FNTimeOfDayScalar(Night Scale=0.5, Morning Scale=0.6, Evening Scale=0.6)
      B=&Multiply_Float(A=0.14, B=$System.AlphaFade)
     )
    )
    Scale RGB=&Multiply_VectorByFloat(
     Vector=&MakeVectorFromLinearColorRGB(Color=&MultiplyLinearColor(
       Linear Color=&LerpLinearColors(
        StartColor=&ColorFromCurve(ColorCurve=DataInterface() {propertyValues {
           AlphaCurve {keys [<at 0 1 RCIM_Cubic>,<at 1 1 RCIM_Cubic>]}
           BlueCurve {keys [
             <at 0 0.7856>
             {arriveTangent=-2.8488, interpMode=RCIM_Cubic, leaveTangent=-2.8488, time=0.1326, value=0.1248}
             <at 0.2716 0.0119 RCIM_Cubic>
            ]}
           GreenCurve {keys [
             <at 0 4.1318>
             {arriveTangent=-13.7259, interpMode=RCIM_Cubic, leaveTangent=-13.7259, time=0.1326, value=0.7925}
             <at 0.2716 0.4039 RCIM_Cubic>
            ]}
           RedCurve {keys [<at 0 8>, {arriveTangent=-18.4097, interpMode=RCIM_Cubic, leaveTangent=-18.4097, time=0.1326, value=3}, <at 0.2716 3 RCIM_Cubic>]}
          }})
        EndColor=&ColorFromCurve(ColorCurve=DataInterface() {propertyValues {
           AlphaCurve {keys [<at 0 1 RCIM_Cubic>,<at 1 1 RCIM_Cubic>]}
           BlueCurve {keys [<at 0 8>, {arriveTangent=-19.5162, interpMode=RCIM_Cubic, leaveTangent=-19.5162, time=0.1326, value=3}, <at 0.2562 3 RCIM_Cubic>]}
           GreenCurve {keys [
             <at 0 4.8705>
             {arriveTangent=-14.4825, interpMode=RCIM_Cubic, leaveTangent=-14.4825, time=0.1326, value=2.0597}
             <at 0.2562 1.1601 RCIM_Cubic>
            ]}
           RedCurve {keys [
             <at 0 1.9358>
             {arriveTangent=-6.4769, interpMode=RCIM_Cubic, leaveTangent=-6.4769, time=0.1326, value=0.4707}
             <at 0.2562 0.2764 RCIM_Cubic>
            ]}
          }})
        LerpFactor=$User.Variant
       )
       Scale Factor=&MakeVector4_FromLinearColor(LinearColor=&FNTimeOfDayColor(
         Night Color=<rgba 0.7691 0.9321 1 1>
         Morning Color=<rgba 1 0.776 0.776 1>
         Day Color=<rgba 1 0.776 0.776 1>
         Evening Color=<rgba 1 0.776 0.776 1>
        ))
      ))
     Float=&FloatFromCurve(FloatCurve=DataInterface() {propertyValues {Curve {keys [<at 0 4 RCIM_Cubic>,<at 0.2335 1>]}}})
    )
    ScaleA=true
    ScaleRGB=true
   )
   &ScaleRibbonWidth(Ribbon Width Scale=&FloatFromCurve(FloatCurve=DataInterface() {propertyValues {Curve {keys [<at 0 1>,<at 1 0.9>]}}}))
   &DynamicMaterialParameters(
    Index 0 Param 1=&FloatFromCurve(FloatCurve=DataInterface() {propertyValues {Curve {keys [
         <at 0.2889 -2.0689 RCIM_Cubic>
         {arriveTangent=-14.6907, interpMode=RCIM_Cubic, leaveTangent=-14.6907, tangentMode=RCTM_User, time=0.8, value=-7}
        ]}}})
    Module.Index0Param0Write=true
    Module.Index0Param1Write=true
    Module.Index0Param2Write=true
    Module.Index0Param3Write=true
   )
   /SetVariables(Particles.DarkWidth=150) {isSetParameters=true}
   /SetVariables(Particles.DarkColor:LinearColor=&MakeLinearColorFromVectorAndFloat(
     "Vector (RGB)"=&MakeVectorFromLinearColorRGB(Color=&LerpLinearColors(
       StartColor=&ColorFromCurve(ColorCurve=DataInterface() {propertyValues {
          AlphaCurve {keys []}
          BlueCurve {keys [<at 0 0 RCIM_Cubic>,<at 0.4118 0 RCIM_Cubic>]}
          GreenCurve {keys [<at 0 0.0009 RCIM_Cubic>,<at 0.4118 0 RCIM_Cubic>]}
          RedCurve {keys [<at 0 0.0469 RCIM_Cubic>,<at 0.4118 0.0104 RCIM_Cubic>]}
         }})
       EndColor=&ColorFromCurve(ColorCurve=DataInterface() {propertyValues {
          AlphaCurve {keys []}
          BlueCurve {keys [<at 0 0.0469 RCIM_Cubic>,<at 0.4118 0.0104 RCIM_Cubic>]}
          GreenCurve {keys [<at 0 0.0209 RCIM_Cubic>,<at 0.4118 0.0006 RCIM_Cubic>]}
          RedCurve {keys [<at 0 0 RCIM_Cubic>,<at 0.4118 0 RCIM_Cubic>]}
         }})
       LerpFactor=$User.Variant
      ))
     "Float (Alpha)"=&FloatFromCurve(
      FloatCurve=DataInterface() {propertyValues {Curve {keys [
          {arriveTangent=-0.0791, interpMode=RCIM_Cubic, leaveTangent=-0.0791, tangentMode=RCTM_User, time=0, value=1}
          {arriveTangent=-0.4815, interpMode=RCIM_Cubic, leaveTangent=-0.4815, tangentMode=RCTM_User, time=0.4337, value=0.7559}
          <at 0.8 0 RCIM_Cubic>
         ]}}}
      Scale Curve=&Multiply_Float(A=4, B=$System.AlphaFade)
     )
    )) {isSetParameters=true}
  ]
  rendererClasses [&NiagaraRibbonRendererProperties]
  renderers [
   {rendererClass=&NiagaraRibbonRendererProperties}
   {rendererClass=&NiagaraRibbonRendererProperties, rendererIndex=1}
   {rendererClass=&NiagaraRibbonRendererProperties, rendererIndex=2}
  ]
 }
 FN_AnimTrail001 {
  emitterSpawnScript [&FN_RibbonIDOffsetUtility()]
  emitterUpdateScript [
   &EmitterState()
   &FN_SpawnAnimTrail()
  ]
  particleSpawnScript [
   &FN_InitializeAnimTrailBoneIndecies(First Bone or Socket Index=1, Second Bone or Socket Index=2)
   &FN_InitializeAnimTrailBoneIndecies(AnimTrail Selection=1, First Bone or Socket Index=1, Second Bone or Socket Index=0) {enabled=false, module=FN_InitializeAnimTrailBoneIndecies001}
   &FN_InitializeAnimTrail(AnimTrailWidth=1.5, Lifetime=0.3)
   &AddVelocity(Velocity=<0 400 0>) {enabled=false}
  ]
  particleUpdateScript [
   &SolveForcesAndVelocity(Clamp Velocity=false, Limit Acceleration=false) {enabled=false}
   &FN_UpdateAnimTrail() {module=FN_UpdateAnimTrail001}
   &ScaleColor(
    Scale Alpha=&FloatFromCurve(
     FloatCurve=DataInterface() {propertyValues {Curve {keys [<at 0.175 1 RCIM_Cubic>,<at 0.7 0>]}}}
     Scale Curve=&Multiply_Float(A=&Lerp_Float(A=1.2, B=1, Alpha=$User.Variant), B=$System.AlphaFade)
    )
    Scale RGB=&Multiply_Vector(
     A=&VectorFromFloat(Value=&FloatFromCurve(FloatCurve=DataInterface() {propertyValues {Curve {keys [<at 0 4 RCIM_Cubic>,<at 0.2 1>]}}}, Scale Curve=3))
     B=&MakeVectorFromLinearColorRGB(Color=&MultiplyLinearColor(
       Linear Color=&FNTimeOfDayColor(
        Night Color=<rgba 0.584 0.6556 0.8125 1>
        Morning Color=<rgba 1 0.8212 0.7448 1>
        Day Color=<rgba 1 0.9304 0.8906 1>
        Evening Color=<rgba 1 0.843 0.7656 1>
       )
       Scale Factor=&MakeVector4_FromLinearColor(LinearColor=&LerpLinearColors(
         StartColor=&ColorFromCurve(ColorCurve=DataInterface() {propertyValues {
            AlphaCurve {keys [<at 0 1 RCIM_Cubic>,<at 1 0 RCIM_Cubic>]}
            BlueCurve {keys [<at 0 1.483 RCIM_Cubic>,<at 0.4577 0.3501 RCIM_Cubic>]}
            GreenCurve {keys [<at 0 4.3882 RCIM_Cubic>,<at 0.4577 0.8086 RCIM_Cubic>]}
            RedCurve {keys [<at 0 12 RCIM_Cubic>,<at 0.4577 12 RCIM_Cubic>]}
           }})
         EndColor=&ColorFromCurve(ColorCurve=DataInterface() {propertyValues {
            AlphaCurve {keys [<at 0 1 RCIM_Cubic>,<at 1 0 RCIM_Cubic>]}
            BlueCurve {keys [<at 0 12 RCIM_Cubic>,<at 0.4577 12 RCIM_Cubic>]}
            GreenCurve {keys [<at 0 9.7807 RCIM_Cubic>,<at 0.4577 1.9754 RCIM_Cubic>]}
            RedCurve {keys [<at 0 5.1306 RCIM_Cubic>,<at 0.4577 0 RCIM_Cubic>]}
           }})
         LerpFactor=$User.Variant
        ))
      ))
    )
    ScaleA=true
    ScaleRGB=true
   )
   &ScaleRibbonWidth(Ribbon Width Scale=&FloatFromCurve(FloatCurve=DataInterface() {propertyValues {Curve {keys [<at 0 0 RCIM_Constant>,<at 0.1076 1 RCIM_Constant>,<at 0.2156 0 RCIM_Constant>,<at 0.331 1 RCIM_Constant>,<at 0.5 0 RCIM_Constant>]}}})) {enabled=false}
  ]
  rendererClasses [&NiagaraRibbonRendererProperties]
  renderers [
   {rendererClass=&NiagaraRibbonRendererProperties}
   {rendererClass=&NiagaraRibbonRendererProperties, rendererIndex=1}
  ]
 }
 Sparks {
  emitterUpdateScript [
   &EmitterState(Life Cycle Mode=::NewEnumerator1, Loop Behavior=::NewEnumerator1, Loop Delay=0.01, Loop Duration=0.11, UseLoopDelay=true)
   &SpawnRate(Spawn Probability=0.7, SpawnRate=80, Use Spawn Probability=true)
  ]
  particleSpawnScript [
   &InitializeParticle(
    Lifetime Max=0.55
    Lifetime Min=0.35
    Lifetime Mode=::NewEnumerator1
    Mesh Scale Mode=::NewEnumerator1
    Mesh Uniform Scale Max=1.5
    Mesh Uniform Scale Min=0.5
    Sprite Size Max=<2 6>
    Sprite Size Min=<1 2>
    Sprite Size Mode=::NewEnumerator2
    UsePositionOffset=true
    Write Material Random=true
   )
   &SocketLocation(SamplePosOffset=false, Socket Index=3, Socket Reader=$System.Socket Reader, Socket Sampling Mode=::NewEnumerator3, VelocitySpeedLimit=false)
   &ShapeLocation(
    MaskSpawnGroup=false
    Offset Mode=::NewEnumerator0
    Override Seed=false
    Rotation Axis=<0 1 0>
    Rotation Mode=::NewEnumerator1
    Sphere Radius=30
    UseOwnerScale=true
   )
   &AddVelocity(Velocity=&RandomRangeVector(Minimum=<0 0 80>, Maximum=<0 0 100>))
  ]
  particleUpdateScript [
   &ParticleState()
   &ScaleSpriteSize(
    Uniform Curve Scale=0.8
    Uniform Curve Sprite Scale=DataInterface() {propertyValues {Curve {keys [<at 0 0 RCIM_Cubic>,<at 0.2 1 RCIM_Cubic>,<at 1 0 RCIM_Cubic>]}}}
   )
   /SetVariables(Particles.Glow03_Size=&Multiply_Vector2D_ByFloat(
     Float=&FloatFromCurve(
      FloatCurve=DataInterface() {propertyValues {Curve {keys [
          <at 0 0 RCIM_Cubic>
          <at 0.2 1.3 RCIM_Cubic>
          {arriveTangent=-0.1423, interpMode=RCIM_Cubic, leaveTangent=-0.1423, tangentMode=RCTM_User, time=1, value=0}
         ]}}}
      Scale Curve=8
     )
     Vector2D=<0.7 1.5>
    )) {isSetParameters=true}
   &ScaleColor(
    Scale Alpha=&Multiply_Float(
     A=1
     B=&FloatFromCurve(FloatCurve=DataInterface() {propertyValues {Curve {keys [
          {interpMode=RCIM_Cubic, tangentMode=RCTM_User, time=0.1, value=0}
          {arriveTangent=-1.3622, interpMode=RCIM_Cubic, leaveTangent=-1.3622, tangentMode=RCTM_User, time=0.2, value=1}
          {tangentMode=RCTM_User, time=0.8, value=0}
         ]}}})
    )
    Scale RGB=&MakeVectorFromLinearColorRGB(Color=&LerpLinearColors(
      StartColor=&ColorFromCurve(ColorCurve=DataInterface() {propertyValues {
         AlphaCurve {keys [<at 0.0926 1 RCIM_Cubic>,<at 1 0 RCIM_Cubic>]}
         BlueCurve {keys [<at 0 0.3958 RCIM_Cubic>]}
         GreenCurve {keys [<at 0 0.5254 RCIM_Cubic>]}
         RedCurve {keys [<at 0 4 RCIM_Cubic>]}
        }})
      EndColor=&ColorFromCurve(ColorCurve=DataInterface() {propertyValues {
         AlphaCurve {keys [<at 0.0926 1 RCIM_Cubic>,<at 1 0 RCIM_Cubic>]}
         BlueCurve {keys [<at 0 4 RCIM_Cubic>]}
         GreenCurve {keys [<at 0 2.4543 RCIM_Cubic>]}
         RedCurve {keys [<at 0 0.5475 RCIM_Cubic>]}
        }})
      LerpFactor=$User.Variant
     ))
    ScaleA=true
    ScaleRGB=true
   )
   /SetVariables(Particles.Glow03_Color=&MakeLinearColorFromVectorAndFloat(
     "Vector (RGB)"=&MakeVectorFromLinearColorRGB(Color=&MultiplyLinearColor(
       Linear Color=&LerpLinearColors(StartColor=<rgba 1.5 0.1464 0 1>, EndColor=<rgba 0 0.4116 1.5 1>, LerpFactor=$User.Variant)
       Scale Factor=&MakeVector4_FromFloat(Float=1)
      ))
     "Float (Alpha)"=&Multiply_Float(A=0.2, B=1)
    )) {isSetParameters=true}
   &GravityForce(Gravity=&RandomRangeVector(Minimum=<0 0 50>, Maximum=<0 0 -100>))
   &CurlNoiseForce(
    Bias Noise=false
    Noise Strength=&RandomRangeFloat(Minimum=700, Maximum=400)
    Pan Noise=true
    Pan Noise Field=<0 0 1>
   )
   &Drag(Drag=3, Use Linear Drag=true, Use Rotational Drag=true)
   &SolveForcesAndVelocity(Clamp Velocity=false, Limit Acceleration=false)
  ]
  rendererClasses [&NiagaraSpriteRendererProperties]
  renderers [
   {rendererClass=&NiagaraSpriteRendererProperties}
   {rendererClass=&NiagaraSpriteRendererProperties, rendererIndex=1}
  ]
 }
 Trail_Dark {
  emitterUpdateScript [
   &EmitterState(Life Cycle Mode=::NewEnumerator1, Loop Behavior=::NewEnumerator1, Loop Duration=0.8, UseLoopDelay=false)
   &FN_SpawnAnimTrail()
  ]
  particleSpawnScript [
   &FN_InitializeAnimTrailBoneIndecies(Second Bone or Socket Index=2)
   &FN_InitializeAnimTrail(Color=<rgba 1 1 1 0.8>, Lifetime=0.4)
  ]
  particleUpdateScript [
   &FN_UpdateAnimTrail()
   &ScaleColor(
    Scale Alpha=&Multiply_Float(
     A=&FloatFromCurve(
      FloatCurve=DataInterface() {propertyValues {Curve {keys [
          <at 0 1.7371 RCIM_Cubic>
          {arriveTangent=-1.3606, interpMode=RCIM_Cubic, leaveTangent=-1.3606, time=0.2132, value=1.2596}
          {arriveTangent=-0.4634, interpMode=RCIM_Cubic, leaveTangent=-0.4634, tangentMode=RCTM_User, time=0.4416, value=1.1363}
          {arriveTangent=-0.0464, interpMode=RCIM_Cubic, tangentMode=RCTM_Break, time=1, value=0.5443}
         ]}}}
      Scale Curve=&FN_FrontEndFloatModifier(Gameplay=1, Front End=0.8)
     )
     B=&Multiply_Float(
      A=&FNTimeOfDayScalar(Night Scale=0.55, Morning Scale=0.8, Evening Scale=0.7)
      B=&Multiply_Float(A=0.3, B=$System.AlphaFade)
     )
    )
    Scale RGB=&MakeVectorFromLinearColorRGB(Color=&LerpLinearColors(
      StartColor=&ColorFromCurve(ColorCurve=DataInterface() {propertyValues {
         AlphaCurve {keys []}
         BlueCurve {keys [<at 0.6209 0.0017 RCIM_Cubic>]}
         GreenCurve {keys [<at 0.6209 0.0017 RCIM_Cubic>]}
         RedCurve {keys [<at 0.6209 0.0417 RCIM_Cubic>]}
        }})
      EndColor=&ColorFromCurve(ColorCurve=DataInterface() {propertyValues {
         AlphaCurve {keys []}
         BlueCurve {keys [<at 0.6209 0.0417 RCIM_Cubic>]}
         GreenCurve {keys [<at 0.6209 0.0077 RCIM_Cubic>]}
         RedCurve {keys [<at 0.6209 0.0005 RCIM_Cubic>]}
        }})
      LerpFactor=$User.Variant
     ))
    ScaleA=true
    ScaleRGB=true
   )
   /SetVariables(Particles.Color003:LinearColor=&MakeLinearColorFromVectorAndFloat(
     "Vector (RGB)"=&MakeVectorFromLinearColorRGB(Color=&LerpLinearColors(StartColor=<rgba 0.5625 0.4506 0.1606 1>, EndColor=<rgba 0.0649 0.0767 0.9583 1>, LerpFactor=$User.Variant))
     "Float (Alpha)"=&Multiply_Float(
      A=&FloatFromCurve(
       FloatCurve=DataInterface() {propertyValues {Curve {keys [
           <at 0 1.7371 RCIM_Cubic>
           {arriveTangent=-1.3606, interpMode=RCIM_Cubic, leaveTangent=-1.3606, time=0.2132, value=1.2596}
           {arriveTangent=-0.4634, interpMode=RCIM_Cubic, leaveTangent=-0.4634, tangentMode=RCTM_User, time=0.4416, value=1.1363}
           {arriveTangent=-0.0464, interpMode=RCIM_Cubic, tangentMode=RCTM_Break, time=1, value=0.5443}
          ]}}}
       Scale Curve=&FN_FrontEndFloatModifier(Gameplay=1, Front End=0.8)
      )
      B=&Multiply_Float(
       A=&FNTimeOfDayScalar(Night Scale=0.5, Morning Scale=0.6, Evening Scale=0.6)
       B=&Multiply_Float(A=0.27, B=$System.AlphaFade)
      )
     )
    )) {isSetParameters=true}
   /SetVariables(Particles.Width_Red:NiagaraFloat=&Multiply_Float(
     A=&FloatFromCurve(FloatCurve=DataInterface() {propertyValues {Curve {keys [{arriveTangent=-0.0514, time=0, value=1}, {arriveTangent=-1.0387, leaveTangent=-1.0387, time=1, value=0.92}]}}})
     B=70
    )) {isSetParameters=true}
  ]
  rendererClasses [&NiagaraRibbonRendererProperties]
  renderers [{rendererClass=&NiagaraRibbonRendererProperties}]
 }
 Trail_Light {
  emitterUpdateScript [
   &EmitterState(Life Cycle Mode=::NewEnumerator1, Loop Behavior=::NewEnumerator1, Loop Duration=0.8, UseLoopDelay=false)
   &FN_SpawnAnimTrail()
  ]
  particleSpawnScript [
   &FN_InitializeAnimTrailBoneIndecies(Second Bone or Socket Index=2)
   &FN_InitializeAnimTrail(AnimTrailWidth=1.5, Color=<rgba 3 3 3 1>, Lifetime=0.28)
  ]
  particleUpdateScript [
   &FN_UpdateAnimTrail()
   &ScaleColor(
    Scale Alpha=&Multiply_Float(
     A=&FloatFromCurve(
      FloatCurve=DataInterface() {propertyValues {Curve {keys [
          <at 0 1.7371 RCIM_Cubic>
          {arriveTangent=-1.3606, interpMode=RCIM_Cubic, leaveTangent=-1.3606, time=0.2132, value=1.2596}
          {arriveTangent=-0.4634, interpMode=RCIM_Cubic, leaveTangent=-0.4634, tangentMode=RCTM_User, time=0.4416, value=1.1363}
          {arriveTangent=-0.0464, interpMode=RCIM_Cubic, tangentMode=RCTM_Break, time=1, value=0.5443}
         ]}}}
      Scale Curve=&FN_FrontEndFloatModifier(Gameplay=1, Front End=0.8)
     )
     B=&Multiply_Float(
      A=&FNTimeOfDayScalar(Night Scale=0.55, Morning Scale=0.8, Evening Scale=0.7)
      B=&Multiply_Float(A=0.17, B=$System.AlphaFade)
     )
    )
    Scale RGB=&MakeVectorFromLinearColorRGB(Color=&LerpLinearColors(
      StartColor=&ColorFromCurve(ColorCurve=DataInterface() {propertyValues {
         AlphaCurve {keys []}
         BlueCurve {keys [<at 0.0344 0.6919>,<at 0.2934 0.2333 RCIM_Cubic>]}
         GreenCurve {keys [<at 0.0344 0.7594>,<at 0.2934 0.3002 RCIM_Cubic>]}
         RedCurve {keys [<at 0.0344 1>,<at 0.2934 1 RCIM_Cubic>]}
        }})
      EndColor=&ColorFromCurve(ColorCurve=DataInterface() {propertyValues {
         AlphaCurve {keys []}
         BlueCurve {keys [<at 0.0344 1>,<at 0.2934 1 RCIM_Cubic>]}
         GreenCurve {keys [<at 0.0344 0.8533>,<at 0.2934 0.6324 RCIM_Cubic>]}
         RedCurve {keys [<at 0.0344 0.5225>,<at 0.2934 0.0052 RCIM_Cubic>]}
        }})
      LerpFactor=$User.Variant
     ))
    ScaleA=true
    ScaleRGB=true
   )
   &ScaleRibbonWidth(Ribbon Width Scale=&FloatFromCurve(FloatCurve=DataInterface() {propertyValues {Curve {keys [{arriveTangent=-0.0514, time=0, value=1}, {arriveTangent=-1.0387, leaveTangent=-1.0387, time=1, value=0.93}]}}}))
  ]
  rendererClasses [&NiagaraRibbonRendererProperties]
  renderers [
   {rendererClass=&NiagaraRibbonRendererProperties}
   {rendererClass=&NiagaraRibbonRendererProperties, rendererIndex=1}
  ]
 }
}
summary {
 systemName=NS_Pickaxe_BirchPalace_Trail
 emitters [
  {emitterName=FN_AnimTrail001, rendererClasses [&NiagaraRibbonRendererProperties]}
  {emitterName=FN_AnimTrail, rendererClasses [&NiagaraRibbonRendererProperties]}
  {emitterName=Sparks, rendererClasses [&NiagaraSpriteRendererProperties]}
  {emitterName=Trail_Light, rendererClasses [&NiagaraRibbonRendererProperties]}
  {emitterName=Trail_Dark, rendererClasses [&NiagaraRibbonRendererProperties]}
 ]
 userVariables [
  User.bIsFrontEnd {type=NiagaraBool}
  User.bIsFrontEndPreview {type=NiagaraBool}
  User.RibbonIDOffset {type=NiagaraInt32}
  User.SwingSectionIndex {type=NiagaraInt32}
  User.Variant {type=NiagaraFloat, defaultValue=0, description=""}
 ]
}
